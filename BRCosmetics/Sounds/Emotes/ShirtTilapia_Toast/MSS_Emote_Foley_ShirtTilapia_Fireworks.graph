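# One-shot stereo MetaSound source for the firework foley in the ShirtTilapia Toast emote, using emote attenuation and foley sound class.
# Play passes through a short trigger delay into the shared MSS_Cosmetic_Simple_Base graph, fed an array of twelve firework waves with slight pitch and volume modulation ranges.
# That graph supplies both audio channels and the finished trigger.
# A concurrency override caps simultaneous instances, stopping the oldest.
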MetaSoundSource /BRCosmetics/Sounds/Emotes/ShirtTilapia_Toast/MSS_Emote_Foley_ShirtTilapia_Fireworks
inputs [
 UE.Source.OnPlay {type=Trigger}
 Waves {type=WaveAsset:Array}
]
interfaces [UE.OutputFormat.Stereo:1.0, UE.Source.OneShot:1.0, UE.Source:1.1]
outputs [
 "UE.OutputFormat.Stereo.Audio:0" {type=Audio}
 "UE.OutputFormat.Stereo.Audio:1" {type=Audio}
 UE.Source.OneShot.OnFinished {type=Trigger}
]
pages [Default {
  input_defaults {
   UE.Source.OnPlay=null
   Waves [
    &/Emote_ShirtTilapia_Toast_FireWork_01
    &/Emote_ShirtTilapia_Toast_FireWork_02
    &/Emote_ShirtTilapia_Toast_FireWork_03
    &/Emote_ShirtTilapia_Toast_FireWork_04
    &/Emote_ShirtTilapia_Toast_FireWork_05
    &/Emote_ShirtTilapia_Toast_FireWork_06
    &/Emote_ShirtTilapia_Toast_FireWork_07
    &/Emote_ShirtTilapia_Toast_FireWork_08
    &/Emote_ShirtTilapia_Toast_FireWork_09
    &/Emote_ShirtTilapia_Toast_FireWork_10
    &/Emote_ShirtTilapia_Toast_FireWork_11
    &/Emote_ShirtTilapia_Toast_FireWork_12
   ]
  }
  nodes [
   TriggerDelay_0 = UE.Trigger Delay(In=^UE.Source.OnPlay) {also {defaults {Delay Time=0.1}}}
   MSS_Cosmetic_Simple_Base_0 = /FortMetasound/Cosmetics/MSS_Cosmetic_Simple_Base(Waves=^Waves, UE.Source.OnPlay=@TriggerDelay_0.Out) {also {defaults {PitchModulationRange.Max=1.05, PitchModulationRange.Min=0.95, VolumeModulationRange.Max=1.05, VolumeModulationRange.Min=0.95}}}
  ]
  links [
   {dst="@UE.OutputFormat.Stereo.Audio:0", src=MSS_Cosmetic_Simple_Base_0.UE.OutputFormat.Stereo.Audio:0}
   {dst="@UE.OutputFormat.Stereo.Audio:1", src=MSS_Cosmetic_Simple_Base_0.UE.OutputFormat.Stereo.Audio:1}
   {dst="@UE.Source.OneShot.OnFinished", src=MSS_Cosmetic_Simple_Base_0.UE.Source.OneShot.OnFinished}
  ]
 }]
properties {
 assetImportData=&self:AssetImportData
 attenuationSettings=&Fort_Emotes_Athena_Att
 audioPropertiesSheet=/BRCosmetics/Sounds/Settings/PropertySheets/Emotes/APS_Cosmetic_Emote_Foley
 bOverrideConcurrency=true
 maxDistance=1600
 numChannels=2
 outputFormat=Stereo
 sampleRate=48000
 soundClassObject=/Game/Sounds/FortSoundClassesAndModes/Fort_Foley_Emote
 concurrencyOverrides {maxCount=3, resolutionRule=StopOldest}
}
