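# Niagara effect for the ShirtTilapia Toast emote that surrounds the player with a miniature city and looping fireworks.
# Buildings places meshes from authored index and offset arrays on an owner-centered grid, scaling them through a region-mask-driven Timeline that speeds up with owner movement.
# SingleLoopingParticle launches timed rockets with audio and records their deaths in a system position array.
# The system reads those deaths to trigger DirectionalBurst's colored explosions at the recorded positions.
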NiagaraSystem /BRCosmetics/Animation/Game/MainPlayer/Emotes/ShirtTilapia_Toast/FX/NS_Emote_ShirtTilapiaToast_City
dependencies {
 usedDataInterfaces [
  &NiagaraDataInterfaceArrayInt32
  &NiagaraDataInterfaceCurve
  &NiagaraDataInterfaceMeshRendererInfo
  &NiagaraDataInterfaceArrayFloat3
  &NiagaraDataInterfaceCollisionQuery
  &NiagaraDataInterfaceDebugDraw
  &NiagaraDataInterfaceArrayFloat
  &NiagaraDataInterfaceAudioPlayer
  &NiagaraDataInterfaceArrayColor
  &NiagaraDataInterfaceArrayPosition
  &NiagaraDataInterfaceSocketReader
 ]
 usedDynamicInputs [
  &MakeIntFromFloat
  &MakeVector4_FromFloat
  &FNTimeOfDayScalar
 ]
 usedModules [
  &self:SystemSpawnScript.SystemScriptSource.SystemScriptGraph.NiagaraNodeAssignment_5.SetVariables_#1
  &self:SystemSpawnScript.SystemScriptSource.SystemScriptGraph.NiagaraNodeAssignment_0.SetVariables_#2
  &self:SystemSpawnScript.SystemScriptSource.SystemScriptGraph.NiagaraNodeAssignment_1.SetVariables_#3
  &SystemState
  &self:SystemSpawnScript.SystemScriptSource.SystemScriptGraph.NiagaraNodeAssignment_6.SetVariables_#4
  &self:ReadDeath_01
  &self:SystemSpawnScript.SystemScriptSource.SystemScriptGraph.NiagaraNodeAssignment_2.SetVariables_#5
 ]
}
emitters {
 Buildings {
  emitterUpdateScript [
   &EmitterState(
    Inactive Response=::NewEnumerator1
    Life Cycle Mode=::NewEnumerator1
    Loop Behavior=::NewEnumerator1
    Loop Duration Mode=::NewEnumerator1
    UseLoopDelay=false
   )
   /Niagara/Modules/Spawn/Location/SpawnParticlesInGrid(X Count=12, Y Count=12, Z Count=0)
   /SetVariables(
    Emitter.OriginOffset:Vector3f=&TransformVector(Vector=<0 10 50>, Destination Space=::World)
    Emitter.VolumeRadius=150
   ) {isSetParameters=true}
  ]
  particleSpawnScript [
   &PartitionParticles(Number Of Partitions=4, Partition Mode:ENiagara_PartitionMode=::NewEnumerator5) {module=PartitionParticles001}
   /SetVariables(Particles.MeshIndex=&SelectIntFromArray(
     Direct Array Index=&ReturnExecIndex()
     Int Selection Array=DataInterface() {propertyValues {IntData [23,22,9,2,11,25,14,24,22,21,22,13,25,22,21,23,17,4,11,12,17,22,23,25,24,23,13,18,22,21,23,25,3,7,4,22,11,25,23,25,22,2,3,0,2,4,19,21,19,21,4,3,6,7,9,8,22,13,25,24,23,7,5,1,4,23,21,11,10,21,23,24,22,25,21,22,23,18,20,14,7,25,6,8,17,13,19,12,17,21,24,22,25,9,22,23,21,25,24,23,24,19,23,25,4,19,16,13,24,23,14,19,15,23,24,21,14,18,13,5,22,1,23,20,16,22,18,13,24,15,25,14,23,3,7,10,19,17,25,18,22,10,23,21,20,24,16,18,14,16,11,15,19,11,15,14,23,14,5,21,22,24,18,17,13,23,21,20,19]}}
    )) {isSetParameters=true}
   &FN_SetSpawnGroup(Number of Spawn Groups=6) {module=FN_SetSpawnGroup001}
   &InitializeParticle(
    Mesh Scale Mode=::NewEnumerator3
    Mesh Uniform Scale=0.6
    Position Offset Coordinate Space=::Local
    UsePositionOffset=true
    Write Material Random=true
   )
   /Niagara/Modules/Spawn/Location/V2/GridLocation(
    Coordinate Space:ENiagaraCoordinateSpace=::World
    Module.Jitter Inside Grid Cells=false
    Module.Override Random Seed=false
    Normalize Offsets=false
    Offset:Vector3f=&SelectVectorFromArray(
     Direct Array Index=&ReturnExecIndex()
     Vector Selection Array=DataInterface() {propertyValues {FloatData [
        <2.8557 -16.2123 0>
        <-7.8913 -14.4849 0>
        <7.9754 -4.7757 0>
        <-3.5647 -4.4494 0>
        <15.4042 -8.4087 0>
        <12.1543 -10.7815 0>
        <2.2669 -11.0059 0>
        <0 7.0074 0>
        <-10.0591 2.8405 0>
        <26.4752 0 0>
        <19.674 -5.853 0>
        <11.4691 0 0>
        <0 11.3964 0>
        <-16.7326 4.9635 0>
        <-14.6023 8.3597 0>
        <0 10.0862 0>
        <0 10.775 0>
        <0 11.2125 0>
        <0.9557 11.3088 0>
        <11.0528 0 0>
        <0 0.22 0>
        <15.0746 9.7032 0>
        <6.2809 9.0139 0>
        <-4.4929 9.3542 0>
        <0 -0.22 0>
        <-10.1222 -1.8312 0>
        <-7.7528 -0.5579 0>
        <3.4222 1.0576 0>
        <0 0 0>
        <0 0 0>
        <0 1.7458 0>
        <10.3321 1.9672 0>
        <0.1517 1.1167 0>
        <9.0142 -0.9315 0>
        <-1.6036 -1.194 0>
        <-11.1227 0.2431 0>
        <11.9166 8.5743 0>
        <6.4909 9.5546 0>
        <-0.26 7.8265 0>
        <4.5035 9.1836 0>
        <-6.4434 10.3014 0>
        <-8.7143 8.2674 0>
        <-0.6 8.564 0>
        <-10.3494 6.8282 0>
        <-19.33 7.4526 0>
        <0 9.3739 0>
        <-12.282 9.8394 0>
        <-12.8178 10.7578 0>
        <0 0 0>
        <0 -3.6785 0>
        <-2.7465 -3.2974 0>
        <3.9901 2.2599 0>
        <1.6758 -1.0845 0>
        <-8.211 -1.1444 0>
        <4.8022 -2.6498 0>
        <8.9 -9.1452 0>
        <0.076 53.082 0>
        <7.8961 -0.0503 0>
        <0 -0.88 0>
        <-9.9595 -3.642 0>
        <0 10.1576 0>
        <-9.4082 7.9079 0>
        <-4.7681 9.4902 0>
        <6.9106 7.7746 0>
        <0 10.4941 0>
        <-9.7115 11.0079 0>
        <12.4258 6.8087 0>
        <-0.2768 3.8626 0>
        <6.6292 1.3606 0>
        <0.6837 9.0592 0>
        <5.7469 6.0263 0>
        <0 4.378 0>
        <0 0 0>
        <-9.791 0 0>
        <-10.9379 0 0>
        <0 0 0>
        <-6.3185 0 0>
        <-11.2056 0 0>
        <30.7332 19.2895 0>
        <0 0 0>
        <13.9322 -7.085 0>
        <-0.8151 7.2662 0>
        <2.7411 -1.6279 0>
        <-0.0497 0 0>
        <0 4.378 0>
        <1.7851 3.9467 0>
        <-6.7731 0.7528 0>
        <0 8.4082 0>
        <3.6353 0 0>
        <15.9849 0 0>
        <0 9.3365 0>
        <8.5654 12.0404 0>
        <0 -6.4715 0>
        <-6.3446 -4.5837 0>
        <6.2145 -6.5663 0>
        <0 -3.3255 0>
        <0 -4.8884 0>
        <0 -3.8228 0>
        <-6.4201 -5.361 0>
        <14.9888 -7.4706 0>
        <5.7 -7.5242 0>
        <-4.3141 -4.1731 0>
        <-15.8291 -10.4813 0>
        <0 -0.4614 0>
        <-2.0282 0 0>
        <14.5984 16.4379 0>
        <7.2865 7.6422 0>
        <0 4.554 0>
        <0.2276 9.1013 0>
        <-8.3532 9.565 0>
        <7.3298 11.2925 0>
        <0 10.3585 0>
        <0 6.3827 0>
        <-7.6388 8.624 0>
        <-17.5293 9.7201 0>
        <14.6548 0 0>
        <0 10.7751 0>
        <16.1623 9.1093 0>
        <6.4082 0 0>
        <0.7077 1.4973 0>
        <1.9237 9.2092 0>
        <-4.6293 0 0>
        <7.2363 2.407 0>
        <0 0.2633 0>
        <0 -0.7773 0>
        <-5.3742 0 0>
        <-10.1049 -0.9913 0>
        <8.0743 -13.8397 0>
        <0 0 0>
        <17.3548 0.6602 0>
        <9.2549 -1.9082 0>
        <0.1651 -0.9449 0>
        <0 0 0>
        <-5.8658 -7.3074 0>
        <7.6339 0 0>
        <0 0 0>
        <0 0.2625 0>
        <0 0 0>
        <0 0 0>
        <7.7405 -8.6428 0>
        <-0.65 -9.4904 0>
        <21.0069 11.3124 0>
        <34.4848 12.8133 0>
        <-6.265 -16.5879 0>
       ]}}
    )
    XYZ Dimensions:Vector3f=&MakeVector(
     X=&Divide_Float(A=$Emitter.VolumeRadius, B=6)
     Y=&Divide_Float(A=$Emitter.VolumeRadius, B=6)
     Z=0
    )
   )
   &InitialMeshOrientation(
    Mesh Orientation Mode=::NewEnumerator2
    Rotation=&SelectVectorFromArray(
     Array Sampling Mode=::NewEnumerator0
     Direct Array Index=?
     Vector Selection Array=DataInterface() {propertyValues {FloatData [<0 0 0>, <0 0 0.25>, <0 0 0.5>]}}
     Random Seed=0
    )
    Use Rotation Vector=true
   )
   &DynamicMaterialParameters(
    Index 0 Param 1=&MakeFloatFromLinearColor(
     LinearColor=&FNTimeOfDayColor(Night Color=<rgba 6 6 6 1>, Morning Color=<rgba 0 0 0 1>, Day Color=<rgba 0 0 0 1>, Evening Color=<rgba 1 1 1 1>)
     Channel:ENiagaraFloatFromLinearColorOptions=::NewEnumerator0
    )
    Module.Index0Float0Write=true
    Module.Index0Float1Write=false
    Module.Index0Float2Write=false
    Module.Index0Float3Write=false
   ) {enabled=false}
  ]
  particleUpdateScript [
   &ParticleState(Kill Particles When Lifetime Has Elapsed=false, Let Infinitely Lived Particles Die When Emitter Deactivates=true)
   &SocketLocation(
    Orientation Sampling=::NewEnumerator1
    Position Sampling=::NewEnumerator1
    SamplePosOffset=false
    Socket Reader=$System.Socket Reader
    Socket Sampling Mode=::NewEnumerator3
    VelocitySpeedLimit=false
   )
   &SolveForcesAndVelocity(Clamp Velocity=false, Limit Acceleration=false)
   /SetVariables(Particles.PositionCenter:NiagaraPosition=&AddVectorToPosition(Position=$Engine.Owner.Position, Vector=$Emitter.OriginOffset)) {isSetParameters=true}
   &self:ScratchModule_02(
    GridSize:NiagaraFloat=&Multiply_Float(A=$Emitter.VolumeRadius, B=2)
    InputPosition:NiagaraPosition=$Particles.PositionCenter
   ) {module=ScratchModule_02}
   /SetVariables(Particles.Position=&MakeVector(
     X=&MakeFloatFromVector(Vector=Linked() {linkedVariable {name=Particles.Position, type=NiagaraPosition}}, Channel=::NewEnumerator0)
     Y=&MakeFloatFromVector(Vector=Linked() {linkedVariable {name=Particles.Position, type=NiagaraPosition}}, Channel=::NewEnumerator1)
     Z=&MakeFloatFromVector(Vector=/Niagara/DynamicInputs/Helpers/SimulationPosition(), Channel=::NewEnumerator2)
    )) {isSetParameters=true}
   /SetVariables(Particles.Position=&AddVectorToPosition(Position=$Particles.Position, Vector=<0 0 100>)) {isSetParameters=true}
   /SetVariables(Particles.Position=&AddVectorToPosition(
     Position=$Particles.Position
     Vector=&FN_FrontEndVector3Modifier(
      Gameplay=<0 0 0>
      Front End=&RotateVectorByQuaternion(
       Vector To Rotate=&Multiply_VectorByFloat(Vector=<0 -30 0>, Float=$Engine.DeltaTime)
       Quaternion=$Output.SocketLocation.SampledRotation
      )
     )
    )) {isSetParameters=true}
   &self:ScratchModule_05(TraceDir:Vector3f=<0 0 -1>, TraceDist=250, bIsFrontEnd:NiagaraBool=$User.bIsFrontEndPreview) {module=ScratchModule_05}
   &RegionMask(
    Invert Region=true
    Offset:Vector3f=$Emitter.OriginOffset
    Region Falloff=0
    Region Radius:NiagaraFloat=&Add_Float(A=220, B=-110)
    UseOffset=true
    UseRotation=false
   )
   &Timeline(
    Additional Output:ENiagara_TimelineOutput=::NewEnumerator1
    Additional Output Mode:ENiagara_TimelineOutputMode=::NewEnumerator1
    Float Curve=DataInterface() {propertyValues {Curve {keys [
        {arriveTangent=4.468, interpMode=RCIM_Cubic, leaveTangent=4.468, tangentMode=RCTM_User, time=0.35, value=0}
        {arriveTangent=2.875, interpMode=RCIM_Cubic, leaveTangent=2.875, time=0.5548, value=0.9241}
        {arriveTangent=-0.0156, interpMode=RCIM_Cubic, leaveTangent=-0.0156, tangentMode=RCTM_User, time=0.75, value=1.15}
        {arriveTangent=-0.0297, interpMode=RCIM_Cubic, leaveTangent=-0.0297, tangentMode=RCTM_User, time=0.9, value=0.95}
        {arriveTangent=0.0389, interpMode=RCIM_Cubic, leaveTangent=0.0389, tangentMode=RCTM_User, time=0.96, value=1.025}
        <at 1 1 RCIM_Cubic>
       ]}}}
    Play=&Boolean_NotOperation(Bool=$Particles.RegionMask.IsInsideMask)
    Play Rate=&MakeCustomFloatFromBool(
     Bool=&SetBoolByFloatComparison(A=&VectorLength(Vector=$User.OwnerLocalVelocity), B=160)
     True Float=10
     False Float=1.5
    )
    Rewind Rate:NiagaraFloat=&MakeCustomFloatFromBool(
     Bool=&SetBoolByFloatComparison(A=&VectorLength(Vector=$User.OwnerLocalVelocity), B=160)
     True Float=10
     False Float=1.5
    )
   )
   &ScaleMeshSize(Scale Factor=&VectorFromFloat(Value=&Multiply_Float(
      A=$Particles.Timeline.Float
      B=&FloatFromCurve(FloatCurve=DataInterface() {propertyValues {Curve {
          defaultValue=1
          keys [{interpMode=RCIM_Cubic, tangentMode=RCTM_User, time=0.3101, value=0}, {interpMode=RCIM_Cubic, tangentMode=RCTM_User, time=1, value=1}]
         }}})
     ))) {module=ScaleMeshSize001}
  ]
  rendererClasses [&NiagaraMeshRendererProperties]
  renderers [{rendererClass=&NiagaraMeshRendererProperties, rendererIndex=0}]
 }
 DirectionalBurst {
  emitterUpdateScript [
   &EmitterState(Life Cycle Mode=::NewEnumerator1, Loop Duration=1.1963, UseLoopDelay=true)
   &SpawnPerFrame(Spawn:NiagaraBool=$System.DeathSpawn, Use Spawn Probability=false)
   &SpawnPerFrame(Spawn:NiagaraBool=$System.DeathSpawn, Spawn Group=1, Use Spawn Probability=false) {module=SpawnPerFrame001}
   &SpawnPerFrame(
    Spawn:NiagaraBool=$System.DeathSpawn
    Spawn Count:NiagaraInt32=&RandomRangeInt(Minimum=0, Maximum=1, Evaluation Type=::NewEnumerator1)
    Spawn Group=2
    Use Spawn Probability=false
   ) {module=SpawnPerFrame002}
   &SpawnPerFrame(
    Spawn:NiagaraBool=$System.DeathSpawn
    Spawn Count:NiagaraInt32=&RandomRangeInt(Minimum=10, Maximum=15)
    Spawn Group=3
    Use Spawn Probability=false
   ) {module=SpawnPerFrame003}
  ]
  particleSpawnScript [
   &PartitionParticles(
    Particles Per Partition=1
    Partition Count Mode:ENiagara_InfiniteLoopDuration=::NewEnumerator1
    Partition Mode:ENiagara_PartitionMode_v2=::NewEnumerator1
   )
   /SetVariables(Particles.VisibilityTag=&SelectIntFromArray(Direct Array Index=&SpawnGroup(), Int Selection Array=DataInterface() {propertyValues {IntData [0,1,0,3]}})) {isSetParameters=true}
   &InitializeParticle(
    Color=&MaskLinearColorBySpawnGroup(
     Masked Linear Color=&SelectLinearColorFromArray(
      Array Sampling Mode=::NewEnumerator0
      Direct Array Index=?
      Color Selection Array=DataInterface() {propertyValues {ColorData [<rgba 1 0.6871 0 1>, <rgba 1 1 1 1>, <rgba 0 0.074 1 1>, <rgba 1 0.0069 0 1>]}}
      Random Seed=0
     )
     Passthrough Linear Color=&MaskLinearColorBySpawnGroup(
      Masked Linear Color=<rgba 1 1 1 1>
      Passthrough Linear Color=&SelectLinearColorFromArray(
       Array Sampling Mode=::NewEnumerator1
       Direct Array Index=$Particles.PartitionParticles.Partition
       Direct Array Mode=::NewEnumerator1
       Color Selection Array=DataInterface() {propertyValues {ColorData [<rgba 1 0.0069 0 1>, <rgba 1 0.0069 0 1>, <rgba 0 0.074 1 1>, <rgba 0 0.074 1 1>]}}
       Random Seed=?
      )
      Spawn Group Mask=3
     )
     Spawn Group Mask=2
    )
    Lifetime Max=1
    Lifetime Mode=::NewEnumerator1
    Mass Max=2
    Mass Mode=::NewEnumerator3
    Mesh Scale Mode=::NewEnumerator1
    Mesh Uniform Scale Max=0.002
    Mesh Uniform Scale Min=0.005
    Position=&TransformPosition(Position=$System.DeathPosition, Source Space=::Local, Destination Space=::World)
    Position Mode=::NewEnumerator1
    Sprite Size Mode=::NewEnumerator1
    Uniform Sprite Size Max=1
    Uniform Sprite Size Min=2
    UsePositionOffset=false
    Write Material Random=true
   )
   /SetVariables(
    Particles.Flare Lifetime:NiagaraFloat=&RandomRangeFloat(Minimum=0.15, Maximum=0.2)
    Particles.Flare Sprite Size:Vector2f=&Vector2DFromFloat(Value=&RandomRangeFloat(Minimum=60, Maximum=80))
   ) {isSetParameters=true}
   /SetVariables(Particles.Embers Lifetime:NiagaraFloat=&RandomRangeFloat(Minimum=0.5, Maximum=0.3)) {isSetParameters=true}
   &InitialMeshOrientation(Use Rotation Vector=false)
   &ShapeLocation(
    MaskSpawnGroup=false
    Override Seed=false
    Sphere Radius=&MaskFloatBySpawnGroup(Masked Float=10, Passthrough Float=0, Spawn Group Mask=3)
    UseOwnerScale=true
   )
   &AddVelocity(
    Module.Use Falloff Distance=false
    Override Seed=false
    Velocity Mode=::NewEnumerator1
    Velocity Origin=$Particles.ShapeLocation.ShapePosition
    Velocity Speed=&MaskFloatBySpawnGroup(Masked Float=&RandomRangeFloat(Minimum=80, Maximum=150), Passthrough Float=0, Spawn Group Mask=3)
   )
  ]
  particleUpdateScript [
   &ParticleState()
   /SetVariables(
    Particles.Update Burst Position:NiagaraPosition=$Particles.Position
    Particles.Update Burst Velocity:Vector3f=$Engine.Owner.Velocity
    Particles.Update Embers Age:NiagaraFloat=&Divide_Float(A=$Particles.Age, B=$Particles.Initial.Embers Lifetime)
    Particles.Update Flare Age:NiagaraFloat=&Divide_Float(A=$Particles.Age, B=$Particles.Initial.Flare Lifetime)
   ) {isSetParameters=true}
   &GravityForce(Gravity=&Multiply_VectorByFloat(
     Vector=&RandomRangeVector(Minimum=<0 0 -30>, Maximum=<0 0 -20>)
     Float=&FloatFromCurve(
      FloatCurve=DataInterface() {propertyValues {Curve {
         defaultValue=1
         keys [{interpMode=RCIM_Cubic, tangentMode=RCTM_User, time=0, value=0}, {interpMode=RCIM_Cubic, tangentMode=RCTM_User, time=0.5922, value=1}]
        }}}
      Scale Curve=15
     )
    ))
   &AccelerationForce(Acceleration=&RandomRangeVector(Minimum=<0 0 -25>, Maximum=<0 0 -50>))
   &Drag(
    Drag=&MaskFloatBySpawnGroup(Masked Float=1, Passthrough Float=10, Spawn Group Mask=3)
    Use Linear Drag=true
    Use Rotational Drag=false
   )
   &ScaleColor(
    Scale Alpha=&FloatFromCurve(
     FloatCurve=DataInterface() {propertyValues {Curve {keys [
         {arriveTangent=-4.9376, interpMode=RCIM_Cubic, leaveTangent=7.1838, tangentMode=RCTM_User, time=0, value=1}
         {arriveTangent=-2.4774, interpMode=RCIM_Cubic, leaveTangent=-2.4774, tangentMode=RCTM_User, time=0.1389, value=1.3591}
         {arriveTangent=-0.8467, interpMode=RCIM_Cubic, leaveTangent=-0.8467, tangentMode=RCTM_User, time=0.4181, value=0.6281}
         {arriveTangent=-2.194, interpMode=RCIM_Cubic, leaveTangent=-2.1946, tangentMode=RCTM_User, time=1, value=0}
        ]}}}
     Scale Curve=&Multiply_Float(
      A=&MakeCustomFloatFromBool(
       Bool=&SetBoolByFloatComparison(A=&VectorLength(Vector=$User.OwnerLocalVelocity), B=150)
       True Float=&FloatFromCurve(FloatCurve=DataInterface() {propertyValues {Curve {
           defaultValue=1
           keys [{interpMode=RCIM_Cubic, tangentMode=RCTM_User, time=0, value=1}, {interpMode=RCIM_Cubic, tangentMode=RCTM_User, time=0.468, value=0}]
          }}})
       False Float=1
      )
      B=1
     )
    )
    Scale RGB=&VectorFromFloat(Value=3)
    ScaleA=true
    ScaleRGB=true
   ) {module=ScaleColor002}
   &SolveForcesAndVelocity(Clamp Velocity=false, Limit Acceleration=false, Position:NiagaraPosition=$Particles.Update Burst Position)
   &ScaleSpriteSizeBySpeed(
    Max Scale Factor=&MaskVector2dBySpawnGroup(Masked Vector2D=<0.5 5>, Passthrough Vector2D=<1 1>, Spawn Group Mask=3)
    Min Scale Factor=<0.5 2>
   )
   &ScaleMeshSize(Scale Factor=&VectorFromFloat(Value=&FloatFromCurve(FloatCurve=DataInterface() {propertyValues {Curve {keys [
          {arriveTangent=4.6803, interpMode=RCIM_Cubic, leaveTangent=4.6803, tangentMode=RCTM_User, time=0, value=0}
          {arriveTangent=1.6276, interpMode=RCIM_Cubic, leaveTangent=1.6276, tangentMode=RCTM_User, time=0.1923, value=0.5575}
          {arriveTangent=0.2011, interpMode=RCIM_Cubic, leaveTangent=0.5033, tangentMode=RCTM_User, time=1, value=1}
         ]}}})))
   &DynamicMaterialParameters(
    Index 0 Param 1=1
    Index 0 Param 2=1
    Index 0 Param 3=&FloatFromCurve(
     FloatCurve=DataInterface() {propertyValues {Curve {keys [
         {arriveTangent=3.0114, interpMode=RCIM_Cubic, leaveTangent=3.0114, tangentMode=RCTM_User, time=0, value=0}
         {arriveTangent=0.9456, interpMode=RCIM_Cubic, leaveTangent=0.9456, tangentMode=RCTM_User, time=0.1554, value=0.734}
         {arriveTangent=0.3236, interpMode=RCIM_Cubic, leaveTangent=0.7406, tangentMode=RCTM_User, time=1, value=1}
        ]}}}
     Scale Curve=&RandomRangeFloat(Minimum=0.2, Maximum=0.25)
    )
    Index 0 Param 4=&FloatFromCurve(
     FloatCurve=DataInterface() {propertyValues {Curve {keys [
         {arriveTangent=3.4321, interpMode=RCIM_Cubic, leaveTangent=3.4321, tangentMode=RCTM_User, time=0, value=1}
         {arriveTangent=-1.7957, interpMode=RCIM_Cubic, leaveTangent=-1.7957, tangentMode=RCTM_User, time=0.1672, value=1.5}
         {arriveTangent=-0.286, interpMode=RCIM_Cubic, leaveTangent=-0.286, tangentMode=RCTM_User, time=1, value=0.5}
        ]}}}
     Scale Curve=0.5
    )
    Module.Index0Float0Write=true
    Module.Index0Float1Write=true
    Module.Index0Float2Write=true
    Module.Index0Float3Write=true
   )
   /SetVariables(
    Particles.Update Flare Color:LinearColor=&MakeLinearColorFromVectorAndFloat(
     "Vector (RGB)"=&Multiply_Vector(A=<11 11 11>, B=&MakeVectorFromLinearColorRGB(Color=$Particles.Color))
     "Float (Alpha)"=&FloatFromCurve(
      FloatCurve=DataInterface() {propertyValues {Curve {keys [
          {arriveTangent=0.002, interpMode=RCIM_Cubic, time=0.0903, value=0.8191}
          {interpMode=RCIM_Cubic, leaveTangent=-0.8972, time=1.0011, value=0.002}
         ]}}}
      CurveIndex=$Particles.Update Flare Age
      Scale Curve=&Multiply_Float(
       A=&MakeCustomFloatFromBool(
        Bool=&SetBoolByFloatComparison(A=&VectorLength(Vector=$User.OwnerLocalVelocity), B=150)
        True Float=&FloatFromCurve(FloatCurve=DataInterface() {propertyValues {Curve {
            defaultValue=1
            keys [{interpMode=RCIM_Cubic, tangentMode=RCTM_User, time=0, value=1}, {interpMode=RCIM_Cubic, tangentMode=RCTM_User, time=0.468, value=0}]
           }}})
        False Float=1
       )
       B=1
      )
     )
    )
    Particles.Update Flare Sprite Size:Vector2f=&Multiply_Vector2D(
     A=&Vector2DFromFloat(Value=&FloatFromCurve(
       FloatCurve=DataInterface() {propertyValues {Curve {keys [<at 0.0458 1.0493 RCIM_Cubic>, {arriveTangent=-2.2433, interpMode=RCIM_Cubic, leaveTangent=-2.2433, tangentMode=RCTM_User, time=1, value=0}]}}}
       CurveIndex=$Particles.Update Flare Age
      ))
     B=$Particles.Flare Sprite Size
    )
   ) {isSetParameters=true}
   /SetVariables(Particles.Update Embers Color:LinearColor=&MultiplyLinearColor(
     Linear Color=$Particles.Color
     Scale Factor=&MakeVector4_FromLinearColor(LinearColor=&MakeLinearColorFromVectorAndFloat(
       "Vector (RGB)"=<1 1 1>
       "Float (Alpha)"=&FloatFromCurve(
        FloatCurve=DataInterface() {propertyValues {Curve {
           defaultValue=1
           keys [
            {interpMode=RCIM_Cubic, tangentMode=RCTM_User, time=0, value=1}
            {arriveTangent=-3.7935, interpMode=RCIM_Cubic, leaveTangent=-3.7935, tangentMode=RCTM_User, time=1, value=0}
           ]
          }}}
        CurveIndex=$Particles.Update Embers Age
       )
      ))
    )) {isSetParameters=true}
   /SetVariables(Particles.Linear Color=&MaskLinearColorBySpawnGroup(Masked Linear Color=<rgba 0 0 0 1>, Passthrough Linear Color=<rgba 1 1 1 1>)) {isSetParameters=true}
   &KillParticlesInVolume(Box Size:Vector3f=<350 350 500>, Invert Volume=true, Kill Shape=::NewEnumerator1)
  ]
  rendererClasses [&NiagaraMeshRendererProperties, &NiagaraSpriteRendererProperties]
  renderers [
   {rendererClass=&NiagaraMeshRendererProperties, rendererIndex=0}
   {rendererClass=&NiagaraSpriteRendererProperties, rendererIndex=1}
   {rendererClass=&NiagaraSpriteRendererProperties, rendererIndex=2}
  ]
 }
 Light {rendererClasses [], renderers []}
 SingleLoopingParticle {
  emitterSpawnScript [&InitializeAudioPlayer(Audio Player=DataInterface() {propertyValues {
      Attenuation=null
      Concurrency=null
      MaxPlaysPerTick=10
      SoundToPlay=/BRCosmetics/Sounds/Emotes/ShirtTilapia_Toast/MSS_Emote_Foley_ShirtTilapia_Fireworks
      bAllowLoopingOneShotSounds=false
      bLimitPlaysPerTick=true
      bOnlyActiveDuringGameplay=false
      bStopWhenComponentIsDestroyed=true
      ConfigurationUserParameter {parameter {name=None, typeDefHandle {registeredTypeIndex=1}, varData []}}
      ParameterNames []
     }})]
  emitterUpdateScript [
   &EmitterState(Life Cycle Mode=::NewEnumerator1, Loop Duration=3.3, UseLoopDelay=false)
   &SpawnBurst_Instantaneous(Spawn Time=1, Use Spawn Probability=false, UseLoopCountLimit=false)
   &SpawnBurst_Instantaneous(Spawn Count=2, Spawn Group=1, Spawn Time=1.25, Use Spawn Probability=false, UseLoopCountLimit=false) {module=SpawnBurst_Instantaneous001}
   &SpawnBurst_Instantaneous(Spawn Group=2, Spawn Time=2.5, Use Spawn Probability=false, UseLoopCountLimit=false) {module=SpawnBurst_Instantaneous002}
  ]
  particleSpawnScript [
   &InitializeParticle(
    Color=<rgba 5 2.8758 0.1574 1>
    Lifetime Max=0.4
    Lifetime Min=0.3
    Lifetime Mode=::NewEnumerator1
    Position Offset=<0 10 0>
    Sprite Size Mode=::NewEnumerator1
    Uniform Sprite Size Max=30
    Uniform Sprite Size Min=20
    UsePositionOffset=true
    Write Material Random=true
   )
   &AddVelocity(
    Module.Use Falloff Distance=false
    Origin Offset=<0 10 -5>
    Override Seed=false
    Velocity Mode=::NewEnumerator1
    Velocity Speed=&RandomRangeFloat(Minimum=300, Maximum=600)
   )
   &ShapeLocation(Handle Radius=1, Large Radius=150, MaskSpawnGroup=false, Override Seed=false, Shape Primitive=::NewEnumerator3, UseOwnerScale=true)
   /Niagara/Modules/Audio/PlayAudio()
  ]
  particleUpdateScript [
   &ParticleState()
   &self:WriteDeath_01(Position Array:NiagaraDataInterfaceArrayPosition=$System.DeathArray) {module=WriteDeath_01}
   &DynamicMaterialParameters(
    Index 0 Param 1=0.8
    Index 0 Param 2=1
    Index 0 Param 3=1
    Index 1 Param 1=150
    Index 1 Param 4=1
    Module.Index0Float0Write=true
    Module.Index0Float1Write=true
    Module.Index0Float2Write=true
    Module.Index0Float3Write=true
    Module.Index1Float0Write=true
    Module.Index1Float1Write=true
    Module.Index1Float2Write=true
    Module.Index1Float3Write=true
    Write Parameter Index 1=::NewEnumerator1
   )
   &ScaleColor(
    Scale Alpha=&FloatFromCurve(
     FloatCurve=DataInterface() {propertyValues {Curve {keys [
         {arriveTangent=-4.6344, interpMode=RCIM_Cubic, leaveTangent=-4.6344, tangentMode=RCTM_User, time=0.5, value=1}
         {arriveTangent=-0.2005, tangentMode=RCTM_User, time=0.7479, value=0}
        ]}}}
     Scale Curve=&RandomRangeFloat(Minimum=0.75, Maximum=&FN_FrontEndFloatModifier(Gameplay=1.25, Front End=1))
    )
    Scale RGB=&VectorFromFloat(Value=&FloatFromCurve(
      FloatCurve=DataInterface() {propertyValues {Curve {keys [
          {arriveTangent=-1.6398, interpMode=RCIM_Cubic, leaveTangent=-1.6398, tangentMode=RCTM_User, time=0, value=2}
          {arriveTangent=0.009, tangentMode=RCTM_User, time=1, value=1}
         ]}}}
      Scale Curve=&RandomRangeFloat(Minimum=0.75, Maximum=&FN_FrontEndFloatModifier(Gameplay=5, Front End=3))
     ))
    ScaleA=true
    ScaleRGB=true
   )
   &ScaleSpriteSize(Scale Factor=&Vector2DFromFloat(Value=&FloatFromCurve(
      FloatCurve=DataInterface() {propertyValues {Curve {keys [
          {arriveTangent=6.6282, interpMode=RCIM_Cubic, leaveTangent=6.6282, tangentMode=RCTM_User, time=0, value=0}
          <at 0.1 0.4626>
          {arriveTangent=4.0254, interpMode=RCIM_Cubic, leaveTangent=4.0254, time=0.1078, value=0.2796}
          <at 0.15 0.6638>
          <at 0.2 0.2796>
          <at 0.25 0.5723>
          <at 0.3 0.2887>
          <at 0.35 0.8011>
          <at 0.4 0.2887>
          {arriveTangent=-2.7618, leaveTangent=-2.7618, time=0.45, value=1}
          {arriveTangent=1, leaveTangent=1, time=0.55, value=0.3}
          {arriveTangent=-0.75, leaveTangent=-0.75, time=0.65, value=1.15}
          <at 0.75 0.5>
          {arriveTangent=-6.1253, interpMode=RCIM_Cubic, leaveTangent=-6.1253, tangentMode=RCTM_User, time=0.8, value=1.4}
          <at 1 0>
         ]}}}
      Scale Curve=3
     ))) {module=ScaleSpriteSize002}
   &PointForce(
    Force Origin=&AddVectorToPosition(Position=&SimulationPosition(), Vector=<0 0 50>)
    Module.Use Falloff=false
    Module.UseLinearForce=false
   )
   &SolveForcesAndVelocity(Clamp Velocity=false, Limit Acceleration=false)
   &ScaleSpriteSizeBySpeed(Max Scale Factor=<0.05 1.5>, Min Scale Factor=<0.03 1>)
   /SetVariables(
    Particles.CrossColor:LinearColor=&Multiply_LinearColorByFloat(Linear Color=$Particles.Color, Float=0.5)
    Particles.CrossRotation:NiagaraFloat=&RandomRangeFloat(Minimum=0, Maximum=45)
    Particles.CrossSize:Vector2f=&Vector2DFromFloat(Value=&FloatFromCurve(
      FloatCurve=DataInterface() {propertyValues {Curve {keys [
          {arriveTangent=9.7915, interpMode=RCIM_Cubic, leaveTangent=9.7915, tangentMode=RCTM_User, time=0, value=0}
          <at 0.25 1 RCIM_Cubic>
          <at 1 0 RCIM_Cubic>
         ]}}}
      Scale Curve=&RandomRangeFloat(Minimum=100, Maximum=150)
     ))
   ) {isSetParameters=true}
   &KillParticlesInVolume(Origin Offset:Vector3f=<0 5 40>, Sphere Radius=10) {enabled=false}
  ]
  rendererClasses [&NiagaraSpriteRendererProperties]
  renderers [{rendererClass=&NiagaraSpriteRendererProperties}]
 }
 SmokeLWE {rendererClasses [], renderers []}
}
summary {
 systemName=NS_Emote_ShirtTilapiaToast_City
 emitters [
  {emitterName=Buildings, rendererClasses [&NiagaraMeshRendererProperties]}
  {emitterName=SingleLoopingParticle, rendererClasses [&NiagaraSpriteRendererProperties]}
  {
   emitterName=DirectionalBurst
   rendererClasses [&NiagaraMeshRendererProperties, &NiagaraSpriteRendererProperties]
  }
  {emitterName=SmokeLWE, rendererClasses []}
  {emitterName=Light, rendererClasses []}
 ]
 userVariables [
  User.bIsFrontEnd {type=NiagaraBool}
  User.bIsFrontEndPreview {type=NiagaraBool}
  User.OwnerLocalVelocity {type=Vector3f, description="", defaultValue {struct=Vector3f, value=<1 1 1>}}
 ]
}
